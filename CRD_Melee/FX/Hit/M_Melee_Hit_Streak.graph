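# Translucent unlit Niagara sprite material that draws a streak for melee hit effects, tinted by particle color.
# A motion-lines texture, offset per particle by a random value, is shaped by UV gradients, sine falloffs and a squashed sphere mask.
# Dynamic parameters blend soft and contrast-boosted copies of that mask.
# The depth-faded result drives opacity and, raised to a power, emissive through MF_FX_EmissiveAdjustments.
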Material /CRD_Melee/FX/Hit/M_Melee_Hit_Streak
graph {
 properties {bUsedWithNiagaraSprites=true, blendMode=BLEND_Translucent, pixelDepthOffsetMode=PDOM_Legacy, shadingModel=MSM_Unlit, usedShadingModels=Unlit}
 nodes [
  n1 = TextureSampleParameter2D(UVs=Add(A=TextureCoordinate(), B=ParticleRandom())) {
   group=Motion Lines
   parameterName=Motion Lines Texture
   samplerSource=SSM_Wrap_WorldGroupSettings
   texture=/Game/Effects/Fort_Effects/Textures/Outlander/T_InTheZone_LightRays_01
  }
  n2 = TextureCoordinate()
  n3 = ComponentMask(None=@n2) {g=true}
  n4 = Multiply(A=LinearInterpolate(B=@n1.R, Alpha=ScalarParameter() {defaultValue=0.5, parameterName=Texture}) {constA=1}, B=Multiply(A=@n3, B=@n3))
  n5 = OneMinus(None=ComponentMask(None=@n2) {g=true})
  n6 = Multiply(A=Saturate(None=Multiply(A=@n5) {constB=5})) {constB=0.5}
  n7 = ConstantBiasScale(None=Sine(None=Add(A=@n6) {constB=-0.25}))
  n8 = Sine(None=Multiply(A=ComponentMask(None=@n2) {r=true}) {constB=0.5})
  n9 = MaterialFunctionCall(Float2=TextureCoordinate()) {materialFunction=&BreakOutFloat2Components}
  n10 = OneMinus(None=@n9.G)
  n11 = SphereMask(
   A=Multiply(A=AppendVector(A=@n9.R, B=@n10), B=Constant2Vector() {g=0.5, r=1})
   B=Constant2Vector() {r=0.5}
   Radius=Constant() {r=0.5}
   Hardness=Constant()
  )
  n12 = Multiply(
   A=Multiply(A=Multiply(A=@n4, B=@n7), B=Multiply(A=@n8, B=@n8))
   B=LinearInterpolate(B=Power(Base=@n11), Alpha=ScalarParameter() {defaultValue=0.7, parameterName=SphereMask}) {constA=1}
  )
  n13 = DynamicParameter() {defaultValue=<rgba 0 1 0 1>, paramNames [Contrast, Pre Contrast Mult, Hard Component Mult, Soft Component Mult]}
  n14 = MaterialFunctionCall(In=Multiply(A=@n12, B=@n13.Pre Contrast Mult), Contrast=@n13.Contrast) {materialFunction=&CheapContrast}
  n15 = DepthFade(Opacity=Add(A=Multiply(A=@n12, B=@n13.Soft Component Mult), B=Multiply(A=@n14.Result, B=@n13.Hard Component Mult))) {fadeDistanceDefault=35}
  n16 = Power(Base=@n15, Exp=ScalarParameter() {defaultValue=2, parameterName=Color Exp})
  n17 = ParticleColor()
  MaterialFunctionCall(Emissive=Multiply(A=@n17.RGB, B=Add(A=@n16))) {materialFunction=&MF_FX_EmissiveAdjustments}
  Multiply(A=@n17.A, B=@n15)
 ]
}
